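; Looping gameplay cue notify for the Reality Flower seed pod low gravity buff, handling how its audio fades out.
; On activation it sets a timer for the buff duration minus the fade length, which fades out the cue's audio component, then calls the parent.
; When the cue ceases to be relevant, it calls the parent, then spawns a fade-out sound attached to the target's root and immediately starts fading it over its own duration.
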(event Custom|FadeAudio
  (Audio|Components|Audio|FadeOut (Variables|Sounds|GetAudioComponent) (Variables|Default|GetAudioFadeOutDuration) 0.0 "Linear" :!class /Script/Engine.AudioComponent))

(fn Activated
  (params (PlayerPawn :type Actor%20Object%20Reference))
  (Utilities|Time|SetTimerbyEvent (EventDispatchers|CreateEvent :!event FadeAudio) (- (ScalableFloat|GetValueatLevel (Variables|Default|GetBuffDuration)) (Variables|Default|GetAudioFadeOutDuration)))
  (|CallParentFunction PlayerPawn))

(fn OnCeaseRelevant
  (params (MyTarget :type Actor%20Object%20Reference) (Parameters :type Gameplay%20Cue%20Parameters%20Structure%20%28by%20ref%29))
  (bind _returnvalue (|CallParentFunction MyTarget Parameters))
  (bind _returnvalue_2 (Audio|SpawnSoundAttached (Variables|Default|GetFadeOutSound) (Variables|/Script/Engine.Actor|GetRootComponent MyTarget)))
  (Audio|Components|Audio|FadeOut _returnvalue_2 (Variables|Default|GetFadeOutSound_FadeOutDuration) 0.0 "Linear" :!class /Script/Engine.AudioComponent)
  (return _returnvalue))

(fn ConstructionScript)
