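; Gameplay cue that plays impact effects and sounds when a Reality Flower seed pod takes damage.
; On burst, it branches on aggregated source tags for ranged, melee, fire or default damage, with debug prints for each.
; Ranged and melee hits spawn a Niagara impact aligned to the hit normal, with a shotgun flag set for ranged.
; Non-ranged hits also trigger the pod's FX pulse, and default damage reuses the melee sound.
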(fn OnBurst(EitherVFX-preferthisversion)
  (params (MyTarget :type Actor%20Object%20Reference) (Parameters :type Gameplay%20Cue%20Parameters%20Structure%20%28by%20ref%29) (ParticleComponents :type Array%20of%20FXSystem%20Component%20Object%20References) (AudioComponents :type Array%20of%20Audio%20Component%20Object%20References) (BurstCameraShakeInstance :type Camera%20Shake%20Base%20Object%20Reference) (BurstDecalInstance :type Decal%20Actor%20Object%20Reference))
  (bind (_normalizedmagnitude _rawmagnitude _effectcontext _matchedtagname _originaltag _aggregatedsourcetags _aggregatedtargettags _location _normal) (Ability|GameplayCue|BreakGameplayCueParameters Parameters))
  (if (GameplayTags|HasTag _aggregatedsourcetags "(TagName=\"Weapon.Ranged\")")
    (bind _returnvalue (Niagara|SpawnSystematLocation self "/RealitySeedGameplay/Environment/Assets_RealityFlower/FX/NS_RealityPod_Impact.NS_RealityPod_Impact" _location (Math|Conversions|RotationFromXVector _normal) "1.000000,1.000000,1.000000" true true "AutoRelease"))
    (if (GameplayTags|HasTag _aggregatedsourcetags "(TagName=\"Weapon.Ranged.Shotgun\")")
      (Niagara|SetNiagaraVariable(Bool) _returnvalue "IsShotgun" true)
      (else
        (Niagara|SetNiagaraVariable(Bool) _returnvalue "IsShotgun" false)))
    (Audio|PlaySoundatLocation self (Variables|Default|GetBullet%20Impact%20Sound) _location)
    (Development|PrintString 0 "Ranged Damage")
    (return)
    (else
      (bind _asbga_athena_reality_flower_seed_pod (Utilities|Casting|CastToBGA_Athena_RealityFlower_SeedPod MyTarget)
        (:then
          (Class|BGAAthenaRealityFlowerSeedPod|TriggerFXPulse _asbga_athena_reality_flower_seed_pod))
        (:CastFailed
          _))
      (if (GameplayTags|HasTag _aggregatedsourcetags "(TagName=\"Weapon.Melee\")")
        (Audio|PlaySoundatLocation self (Variables|Default|GetMelee%20Impact%20Sound) _location)
        (Development|PrintString 0 "Melee Damage")
        (Niagara|SpawnSystematLocation self "/RealitySeedGameplay/Environment/Assets_RealityFlower/FX/NS_RealityPod_Impact.NS_RealityPod_Impact" _location (Math|Conversions|RotationFromXVector _normal) "1.000000,1.000000,1.000000" true true "AutoRelease")
        (return)
        (elif (GameplayTags|HasTag _aggregatedsourcetags "(TagName=\"Gameplay.Damage.Elemental.Fire\")")
          (Audio|PlaySoundatLocation self (Variables|Default|GetFire%20Impact%20Sound) _location)
          (Development|PrintString 0 "Fire Damage")
          (return)
          (else
            (Audio|PlaySoundatLocation self (Variables|Default|GetMelee%20Impact%20Sound) _location)
            (Development|PrintString 0 "Default Damage")
            (return)))))))
